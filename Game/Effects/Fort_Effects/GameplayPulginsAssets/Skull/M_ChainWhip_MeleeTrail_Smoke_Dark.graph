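# Translucent unlit Niagara ribbon material for a dark smoke chain-whip melee trail, built around a skull texture.
# The texture slides along the ribbon through the AnimUOffet dynamic parameter, with offset, uniform tiling, axis swap and flip controls.
# Its blue channel plus bloom-scaled green blends two color parameters, weighted toward one end and tinted by particle color.
# The same shape times particle alpha, two sphere masks and a sine width fade forms a soft-edged alpha.
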Material /Game/Effects/Fort_Effects/GameplayPulginsAssets/Skull/M_ChainWhip_MeleeTrail_Smoke_Dark
graph {
 properties {bUsedWithNiagaraRibbons=true, blendMode=BLEND_Translucent, pixelDepthOffsetMode=PDOM_Legacy, shadingModel=MSM_Unlit, usedShadingModels=Unlit}
 nodes [
  n1 = VectorParameter() {defaultValue=<rgba 0 0.2135 0.1737 1>}
  n2 = VectorParameter() {parameterName=Param_2, defaultValue=<rgba 0.0744 1 0.3081 1>}
  n3 = TextureCoordinate()
  n4 = MaterialFunctionCall(Float2=LinearInterpolate(A=@n3, B=OneMinus(None=@n3), Alpha=ScalarParameter() {parameterName=RevertU})) {materialFunction=&BreakOutFloat2Components}
  n5 = TextureCoordinate()
  n6 = VectorParameter() {parameterName=UV_Offset_Tiling, defaultValue=<rgba 0.55 0.35 0.56 0.6>}
  n7 = LinearInterpolate(Alpha=DynamicParameter() {defaultValue=<rgba 0.5 1 1 1>, paramNames [AnimUOffet,Param2,Param3,Param4]}.AnimUOffet) {constA=-1.5, constB=1.5}
  n8 = Add(
   A=LinearInterpolate(A=@n5, B=OneMinus(None=@n5), Alpha=ScalarParameter() {parameterName=RevertU})
   B=MaterialFunctionCall(R=Add(A=@n6.R, B=@n7), G=@n6.G) {materialFunction=&AppendMany}.RG
  )
  n9 = MaterialFunctionCall(R=@n6.B, G=@n6.B) {materialFunction=&AppendMany}
  n10 = TextureSampleParameter2D(UVs=MaterialFunctionCall(XY=Multiply(A=@n8, B=@n9.RG)) {materialFunction=&Swizzle}.YX) {parameterName=MainTex, texture=/Game/Gadgets/Assets/VinderTech_GliderChute/Glider_Thrive_Female/FX/Texture/skull}
  n11 = Add(A=@n10.B, B=Multiply(A=@n10.G, B=ScalarParameter() {defaultValue=0.5, parameterName=Bloom}))
  n12 = Multiply(A=Power(Base=OneMinus(None=@n4.R)) {constExponent=4}, B=@n11)
  n13 = ParticleColor()
  MaterialFunctionCall(Emissive=Multiply(A=LinearInterpolate(A=@n1.RGB, B=@n2.RGB, Alpha=@n12), B=@n13.RGB)) {materialFunction=&MF_FX_EmissiveAdjustments}
  MaterialFunctionCall() {materialFunction=&DebugTimeSine}
  n14 = ScalarParameter() {defaultValue=1, parameterName=Param_1}
  n15 = SphereMask(A=TextureCoordinate(), B=AppendVector(A=Constant() {r=0.7}, B=Constant() {r=0.5})) {attenuationRadius=0.7, hardnessPercent=95}
  n16 = Multiply(A=@n4.G) {constB=0.5}
  n17 = Saturate(None=Multiply(A=Sine(None=@n16)) {constB=2})
  n18 = TextureCoordinate()
  n19 = AppendVector(A=Constant() {r=0.3}, B=Constant() {r=0.5})
  Multiply(
   A=Multiply(A=Multiply(A=@n13.A, B=@n14), B=@n11)
   B=Multiply(A=Multiply(A=@n15, B=@n17), B=SphereMask(A=@n18, B=@n19) {attenuationRadius=0.7, hardnessPercent=90})
  )
  ScalarParameter() {defaultValue=0.6}
 ]
}
