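# Translucent, two-sided lit material for Niagara mesh particles, apparently a ghostly skull effect, tinted green by default.
# A radially mapped, panning swirl normal distorts UVs for caustics and soft-flare textures, with dynamic parameters controlling distortion and SkullAnimation blending out random offsets.
# These textures drive Fresnel and vertex-color tinted emissive, plus opacity shaped by inverse Fresnel, particle alpha and depth fade.
# Time-scrolled vector noise, masked by the textures, builds a displacement zeroed on mobile, forward and low quality.
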Material /Game/Effects/Fort_Effects/GameplayPulginsAssets/Skull/M_Skull
graph {
 properties {
  allowTranslucentCustomDepthWrites=true
  bUsedWithNiagaraMeshParticles=true
  blendMode=BLEND_Translucent
  pixelDepthOffsetMode=PDOM_Legacy
  twoSided=true
  usedShadingModels=Default Lit
 }
 nodes [
  ActorPositionWS()
  n1 = LinearInterpolate(Alpha=VertexColor().R) {constA=1.5, constB=7.5}
  n2 = Multiply(A=@n1) {constA=0.65, constB=0.65}
  n3 = PerInstanceRandom()
  n4 = WorldPosition()
  n5 = ObjectPositionWS()
  n6 = ScalarParameter() {defaultValue=2.5, group=Movement, parameterName=Curl_Speed}
  n7 = LinearInterpolate(A=@n6, B=Multiply(A=@n6) {constA=0.65, constB=0.5}, Alpha=PerInstanceRandom())
  n8 = Time()
  n9 = ScalarParameter() {defaultValue=75, group=Movement, parameterName=Curl_Scale}
  n10 = Divide(
   A=Subtract(A=Subtract(A=@n4.XYZ, B=@n5), B=Multiply(A=@n7, B=@n8))
   B=LinearInterpolate(A=@n9, B=Multiply(A=@n9) {constA=0.65, constB=0.5}, Alpha=PerInstanceRandom())
  )
  n11 = Constant()
  n12 = ShadingPathSwitch(
   Default=Multiply(A=LinearInterpolate(A=@n1, B=@n2, Alpha=@n3), B=VectorNoise(World Position=@n10) {noiseFunction=VNF_VectorALU})
   Mobile=@n11
  ) {desc=Migrated from FeatureLevelSwitch}
  n13 = MaterialFunctionCall(Float4=Constant4Vector() {constant=<rgba 1 1 1 1>}.RGBA) {materialFunction=&BreakOutFloat4Components}
  n14 = MaterialFunctionCall() {materialFunction=&VectorToRadialValue}
  n15 = MaterialFunctionCall(Float2=@n14.Radial Coordinates) {materialFunction=&BreakOutFloat2Components}
  n16 = Multiply(
   A=Subtract(A=Multiply(A=@n15.G), B=ScalarParameter() {defaultValue=0.1, parameterName=MaskOffset})
   B=DynamicParameter() {parameterIndex=1, defaultValue=<rgba 0.2 1 1 1>, paramNames [MaskScale,Param2,Param3,Param4]}.MaskScale
  )
  n17 = ParticleRandom()
  n18 = AppendVector(A=@n17, B=@n17)
  n19 = VectorParameter() {parameterName=Speed, defaultValue=<rgba -0.1 -0.2 1 5>}
  n20 = MaterialFunctionCall(UVs=@n14.Radial Coordinates, Texture Scale=AppendVector(A=@n19.B, B=@n19.B)) {materialFunction=&ScaleUVsByCenter}
  n21 = TextureSample(UVs=Panner(Coordinate=Add(A=@n18, B=@n20.UVs), Speed=AppendVector(A=@n19.R, B=@n19.R))) {samplerType=SAMPLERTYPE_Normal, texture=/Game/VisualThreatManager/StormVisuals/LargeClouds/T_StormSwirlNormal}
  n22 = AppendVector(A=@n21.R, B=@n21.R)
  n23 = ScalarParameter() {defaultValue=1, parameterName=Noise2}
  n24 = Add(A=TextureCoordinate(), B=Multiply(A=Saturate(None=@n16), B=Multiply(A=@n22, B=@n23)))
  n25 = Multiply(A=AppendVector(A=@n13.B, B=@n13.B), B=@n24)
  n26 = LinearInterpolate(Alpha=ParticleRandom()) {constB=0.2}
  n27 = Saturate(None=DynamicParameter() {paramNames [SkullAnimation, HEIGHT, ENABLEHEIGHT, Param4]}.SkullAnimation)
  n28 = TextureSample(UVs=LinearInterpolate(A=Add(A=@n25, B=@n26), B=@n24, Alpha=@n27)) {texture=/Game/Effects/Fort_Effects/Textures/Tile/T_ShieldCaustics_Tile_Chromatic}
  n29 = Multiply(A=AppendVector(A=@n13.R, B=@n13.R), B=@n24)
  n30 = LinearInterpolate(Alpha=ParticleRandom()) {constB=0.3}
  n31 = TextureSample(UVs=LinearInterpolate(A=Add(A=@n29, B=@n30), B=@n24, Alpha=@n27)) {texture=/Game/Global/Textures/Default/Masks_and_Gradients/T_New_Soft_Flair}
  n32 = MaterialFunctionCall() {materialFunction=&LinearGradient}
  n33 = Multiply(A=LinearInterpolate(A=@n28.G, B=Add(A=@n31.G, B=@n32.VGradient), Alpha=@n31.G))
  LinearInterpolate(
   A=ShadingPathSwitch(Default=QualitySwitch(Default=@n12, Low=@n11, Medium=@n12), Forward=@n11)
   B=Constant()
   Alpha=Clamp(None=Multiply(A=@n33))
  )
  n34 = VectorParameter() {parameterName=Color3, defaultValue=<rgba 0.0095 0.125 0.0404 1>}
  n35 = Fresnel() {baseReflectFraction=0.05, exponent=0.5}
  n36 = LinearInterpolate(
   A=VectorParameter() {parameterName=Color, defaultValue=<rgba 0.1534 0.4635 0.1521 1>}.RGB
   B=VectorParameter() {parameterName=Color2, defaultValue=<rgba 0.0865 1 0.329 1>}.RGB
   Alpha=@n35
  )
  n37 = VertexColor()
  n38 = Multiply(
   A=Add(A=@n31.RGB, B=@n28.RGB)
   B=Desaturation(None=LinearInterpolate(A=@n34.RGB, B=@n36, Alpha=@n37.G), Fraction=ScalarParameter() {parameterName=Fraction})
  )
  n39 = Multiply(A=ScalarParameter() {defaultValue=0.35, parameterName=Emmisive}, B=@n38)
  MaterialFunctionCall(Emissive=@n39) {materialFunction=&MF_FX_EmissiveAdjustments}
  Multiply(A=ParticleColor().RGB, B=@n38)
  n40 = VertexColor()
  n41 = OneMinus(None=@n35)
  n42 = Multiply(A=@n33, B=ParticleColor().A)
  n43 = Constant() {r=2.5}
  n44 = Multiply(A=Constant() {r=0.75}, B=Multiply(A=Power(Base=@n41) {constExponent=1}, B=Multiply(A=@n42, B=@n43)))
  n45 = DepthFade() {fadeDistanceDefault=10}
  Multiply(A=ParticleColor().A, B=Multiply(A=OneMinus(None=@n40.R), B=Multiply(A=@n44, B=@n45)))
  Multiply(A=@n39)
 ]
}
